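; Fortnite movement-mode extension for clambering, handling montage selection and activation effects.
; Montage choice branches on ledge launch (window, else sideways ledge offset), first or third person, reach-back, and capsule-adjusted target height with window variants.
; Activation plays camera shake and a gameplay cue; authority applies clamber and build-block effects and cancels or blocks tagged abilities, while non-authority destroys the location highlight.
; Deactivation on authority removes those effects and unblocks abilities.
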(fn GetAnimationMontageInformation
  (params (Context_BaseSynchedActionInfo_WarpSyncPoint1Name :type Name) (Context_BaseSynchedActionInfo_WarpSyncPointLandingOrientationName :type Name) (Context_BaseSynchedActionInfo_DistanceToTarget :type Float%20%28single-precision%29) (Context_BaseSynchedActionInfo_DistanceToTargetX :type Float%20%28single-precision%29) (Context_BaseSynchedActionInfo_DistanceToTargetY :type Float%20%28single-precision%29) (Context_BaseSynchedActionInfo_DistanceToTargetZ :type Float%20%28single-precision%29) (Context_BaseSynchedActionInfo_AngleToTargetHorizontal :type Float%20%28single-precision%29) (Context_BaseSynchedActionInfo_DistanceWarpPoint1ToWarpPoint2 :type Float%20%28single-precision%29) (Context_BaseSynchedActionInfo_DistanceWarpPoint1ToWarpPoint2X :type Float%20%28single-precision%29) (Context_BaseSynchedActionInfo_DistanceWarpPoint1ToWarpPoint2Y :type Float%20%28single-precision%29) (Context_BaseSynchedActionInfo_DistanceWarpPoint1ToWarpPoint2Z :type Float%20%28single-precision%29) (Context_BaseSynchedActionInfo_StartingGroundAngle :type Float%20%28single-precision%29) (Context_FutureLedgeWarpPoint :type Vector) (Context_bIsLedgeLaunch :type Boolean) (Context_bHasFutureLedge :type Boolean) (Context_bClamberFromFalling :type Boolean) (Context_bWindowClamber :type Boolean) (Context_bIsReachBack :type Boolean))
  (bind _ownerpawn (Variables|FortMovementMode_BaseExtLogic|GetOwnerPawn)
    (:then
      ; Get Anim Instance to check if in First Person
      (Utilities|IsValid (Components|SkeletalMesh|GetAnimInstance (Variables|/Script/Engine.Character|GetMesh _ownerpawn))
        (:Is%20Valid
          (bind _asfortnite_m_avg_player_anim_blueprint (Utilities|Casting|CastToFortnite_M_Avg_Player_AnimBlueprint (Components|SkeletalMesh|GetAnimInstance (Variables|/Script/Engine.Character|GetMesh _ownerpawn)))
            (:then
              (LocalVariables|SetIs%20First%20Person (Variables|/Script/FortniteGame.FortPlayerAnimInstance|GetbIsInFirstPerson _asfortnite_m_avg_player_anim_blueprint)))
            (:CastFailed
              _)))
        (:Is%20Not%20Valid
          _))
      (LocalVariables|SetIsLedgeLaunch Context_bIsLedgeLaunch)
      (LocalVariables|SetTargetDistanceZ (+ (Components|Capsule|GetScaledCapsuleHalfHeight (Variables|/Script/Engine.Character|GetCapsuleComponent _ownerpawn)) Context_BaseSynchedActionInfo_DistanceToTargetZ))
      (LocalVariables|SetHasFutureLedge Context_bHasFutureLedge)
      (LocalVariables|SetFutureLedgeY (.y Context_FutureLedgeWarpPoint))
      (LocalVariables|SetClamberFromFalling Context_bClamberFromFalling)
      (LocalVariables|SetWindowClamber Context_bWindowClamber)
      (LocalVariables|SetIsReachBack Context_bIsReachBack)
      (if (LocalVariables|GetIsLedgeLaunch)
        (if (LocalVariables|GetWindowClamber)
          ;[1 1 1] Window
          (return (Variables|Animation|GetClamber_150up_Window) "None" "None")
          (elif (and (< (LocalVariables|GetFutureLedgeY) 5e+01) (>= (LocalVariables|GetFutureLedgeY) -5e+01))
            ;[1 1 1] "Ledge launch"
            (return (Variables|Animation|Ledge|GetLedgeLaunch_350up) "None" "None")
            (elif (>= (LocalVariables|GetFutureLedgeY) 0)
              (return (Variables|Animation|Ledge|GetLedgeLaunch_350up_252rgt) "None" "None")
              (else
                (return (Variables|Animation|Ledge|GetLedgeLaunch_350up_252lft) "None" "None")))))
        (elif (LocalVariables|GetIs%20First%20Person)
          ;[1 1 1] "Clamber" 1st Person
          (if (>= (LocalVariables|GetTargetDistanceZ) 175.0)
            (return (select (LocalVariables|GetWindowClamber) (Variables|Animation|GetClamber_250up_Window) (Variables|Animation|GetClamber_250up)) "None" "None")
            (elif (>= (LocalVariables|GetTargetDistanceZ) 1e+02)
              (return (select (LocalVariables|GetWindowClamber) (Variables|Animation|GetClamber_150up_Window) (Variables|Animation|GetClamber_150up_FP)) "None" "None")
              (else
                (return (select (LocalVariables|GetWindowClamber) (Variables|Animation|GetClamber_100up_Window) (Variables|Animation|GetClamber_100up_FP)) "None" "None"))))
          (elif (LocalVariables|GetIsReachBack)
            ; Reach Back
            (return (Variables|Animation|GetClamber_ReachBack) "None" "None")
            (elif (>= (LocalVariables|GetTargetDistanceZ) 175.0)
              ;[1 1 1] "Clamber" 3rd Person
              (return (select (LocalVariables|GetWindowClamber) (Variables|Animation|GetClamber_250up_Window) (Variables|Animation|GetClamber_250up)) "None" "None")
              (elif (>= (LocalVariables|GetTargetDistanceZ) 1e+02)
                (return (select (LocalVariables|GetWindowClamber) (Variables|Animation|GetClamber_150up_Window) (Variables|Animation|GetClamber_150up)) "None" "None")
                (else
                  (return (select (LocalVariables|GetWindowClamber) (Variables|Animation|GetClamber_100up_Window) (Variables|Animation|GetClamber_100up)) "None" "None"))))))))
    (:else)))

(fn OnExtensionActivate
  (params (RuntimeContext :type MMERuntime%20Context%20Structure%20%28by%20ref%29))
  (bind _returnvalue (Ability|GetFortAbilitySystemComponent (Variables|FortMovementMode_BaseExtLogic|GetOwnerPawn)))
  (bind _ownerpawn (Variables|FortMovementMode_BaseExtLogic|GetOwnerPawn))
  (bind _returnvalue_2 (Class|FortMovementModeExtBlueprintLibrary|GetRuntimeDatafromContext RuntimeContext "/Script/ClamberingCodeRuntime.FortMovementMode_ClamberingRuntimeData"))
  (Game|Feedback|ClientStartCameraShake :self (Fortnite|GetFortPlayerControllerfromActor _ownerpawn) :Shake (Variables|Default|GetCamera%20Shake) :Scale 0.5 :PlaySpace "CameraLocal")
  (GameplayCue|ExecuteGameplayCueOnActor(Burst) _ownerpawn (Variables|Default|GetClamber%20Gameplay%20Cue) (Ability|GameplayCue|MakeGameplayCueParameters :NormalizedMagnitude 0.0 :RawMagnitude 0.0 :Location (Class|FortMovementModeClamberingRuntimeData|GetSyncPosition _returnvalue_2) :Normal "0,0,0" :GameplayEffectLevel 1 :AbilityLevel 1 :bReplicateLocationWhenUsingMinimalRepProxy false))
  (if (Networking|HasAuthority _ownerpawn)
    (bind _asfort_ability_system_component (Utilities|Casting|CastToFortAbilitySystemComponent (Ability|GetAbilitySystemComponent _ownerpawn)))
    (bind _returnvalue_3 (GameplayEffects|FortApplyGameplayEffectToSelf :self _asfort_ability_system_component :GameplayEffectClass (Variables|Default|GetGE%20Clambering) :Level 0.0))
    (Utilities|Array|Add (Variables|Default|GetActive%20GEHandle) _returnvalue_3)
    (bind _returnvalue_4 (GameplayEffects|FortApplyGameplayEffectToSelf :self _asfort_ability_system_component :GameplayEffectClass (Variables|Default|GetGE%20Clambering_BlockBuild) :Level 0.0))
    (Utilities|Array|Add (Variables|Default|GetActive%20GEHandle) _returnvalue_4)
    (Tags|CancelAbilitiesWithTags _returnvalue (GameplayTags|MakeGameplayTagContainerfromArray (Variables|Default|GetCancel%20Abilities%20with%20Tags)))
    (Tags|BlockAbilitiesWithTags _returnvalue (GameplayTags|MakeGameplayTagContainerfromArray (Variables|Default|GetBlock%20Abilities%20with%20Tags)))
    (return)
    (else
      (Actor|DestroyActor (Variables|Default|GetClamberLocationHighlight)))))

(fn OnExtensionDeactivate
  (params (RuntimeContext :type MMERuntime%20Context%20Structure%20%28by%20ref%29))
  (bind _ownerpawn (Variables|FortMovementMode_BaseExtLogic|GetOwnerPawn))
  (bind _returnvalue (Ability|GetFortAbilitySystemComponent _ownerpawn))
  (if (Networking|HasAuthority _ownerpawn)
    (for _array_element (Variables|Default|GetActive%20GEHandle)
      (GameplayEffects|RemoveActiveGameplayEffect _returnvalue _array_element -1)
      (Utilities|Array|Clear (Variables|Default|GetActive%20GEHandle))
      (Tags|UnBlockAbilitiesWithTags _returnvalue (GameplayTags|MakeGameplayTagContainerfromArray (Variables|Default|GetBlock%20Abilities%20with%20Tags)) false)
      (return))))
